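# Landscape layer material function that blends a Hermes mossy forest ground layer into incoming terrain material, with desktop, mobile and static mesh outputs.
# World-projected moss textures get noise-driven tint and darkening, with diffuse alpha feeding opacity mask and displacement.
# Height-blend functions composite the layer behind landscape switches named "Asphalt", with mobile using a separate moss function.
# Wetlands-to-MossyForest switches and debug lerps are present but unconnected to outputs.
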MaterialFunction /Game/Terrain/Landscape/Materials/Layers/MF_Hermes_Wetlands_MossyForest
graph {nodes [
  n1 = NamedRerouteUsage()
  n2 = NamedRerouteUsage()
  n3 = NamedRerouteUsage()
  n4 = NamedRerouteUsage()
  n5 = NamedRerouteUsage()
  n6 = MaterialFunctionCall!(InputMaterial=@n1, LayerMaterial=@n2, LayerWeight=@n3, UseContrast=@n4, Contrast=@n5) {materialFunction=/Game/Athena/Helios/Environments/Landscape/Materials/MaterialFunctions_Utilities/MF_HeightBlendLandscapeLayers_OpacityMaskDisplaceCustom}
  FunctionOutput(None=LandscapeLayerSwitch!(LayerUsed=@n6.BlendedMaterial, LayerNotUsed=@n1) {parameterName=Asphalt}) {outputName=BlendedMaterial, sortPriority=-300}
  FunctionOutput(None=ShadingPathSwitch(Default=NamedRerouteUsage(), Mobile=NamedRerouteUsage())) {outputName=LayerMask, sortPriority=-1}
  n7 = NamedRerouteUsage()
  n8 = MaterialFunctionCall!(
   InputMaterial=@n7
   LayerMaterial=NamedRerouteUsage()
   LayerWeight=NamedRerouteUsage()
   UseContrast=NamedRerouteUsage()
   Contrast=NamedRerouteUsage()
  ) {materialFunction=/Game/Athena/Helios/Environments/Landscape/Materials/MaterialFunctions_Utilities/MF_HeightBlendLandscapeLayers_OpacityMask}
  FunctionOutput(None=LandscapeLayerSwitch(LayerUsed=@n8.BlendedMaterial, LayerNotUsed=@n7) {parameterName=Asphalt}) {outputName=MobileBlendedMaterial, sortPriority=-100}
  n9 = NamedRerouteUsage()
  n10 = StaticBool()
  FunctionOutput(None=ShadingPathSwitch(
    Default=NamedRerouteUsage()
    Mobile=MaterialFunctionCall!(MobileMaterial=@n9, TransformTangents=@n10) {materialFunction=/Game/Athena/Helios/Environments/Landscape/Materials/MaterialFunctions_Utilities/MF_MobileDeriveNormals}.ResultMaterial
   )) {outputName=StaticMeshMaterial, sortPriority=10}
  n11 = ScalarParameter() {defaultValue=2500, group=Wetlands_MossyForest, parameterName=Moss_Noise1_WorldScale, sortPriority=-500}
  n12 = Add(
   A=TextureSample(UVs=MaterialFunctionCall!(Scale=@n11) {materialFunction=/Game/Athena/Apollo/Environments/Landscape/Materials/Biome/CommonMaterials/MF_SimpleWorldUV}.Result) {samplerSource=SSM_Wrap_WorldGroupSettings, texture=/Game/Environments/Helios/Terrain/Textures/T_Helios_Landscape_Noise}.B
   B=ScalarParameter() {group=Wetlands_MossyForest, parameterName=Moss_Noise1_Add, sortPriority=-500}
  )
  n13 = MaterialFunctionCall(Scale=ScalarParameter() {defaultValue=5100, group=Wetlands_MossyForest, parameterName=Moss_Noise2_WorldScale02, sortPriority=-500}) {materialFunction=&MF_SimpleWorldUV}
  n14 = MaterialFunctionCall(Scale=ScalarParameter() {defaultValue=600, group=Wetlands_MossyForest, parameterName=Moss_WorldScale, sortPriority=-500}) {materialFunction=&MF_SimpleWorldUV}
  n15 = TextureSampleParameter2D(UVs=@n14.Result) {
   group=MossyForest
   parameterName=Hermes_Moss_Diffuse
   samplerSource=SSM_Wrap_WorldGroupSettings
   texture=/Game/Environments/GlobalTextures/T_Hermes_MossyGround_01_D
  }
  n16 = Multiply(
   A=Multiply(
    A=Saturate(None=@n12)
    B=TextureSample(UVs=@n13.Result) {samplerSource=SSM_Wrap_WorldGroupSettings, texture=&T_Helios_Landscape_Noise}.G
   )
   B=Power(Base=@n15.A, Exp=ScalarParameter() {defaultValue=2, group=Wetlands_MossyForest, parameterName=Moss_Noise1_Pow1, sortPriority=-500})
  )
  n17 = VectorParameter() {parameterName=Moss_Noise_Color01, defaultValue=<rgba 10 3.2687 1.0556 1>}
  n18 = OneMinus(None=Power(
    Base=OneMinus(None=@n16)
    Exp=ScalarParameter() {defaultValue=75, group=Wetlands_MossyForest, parameterName=Moss_Noise1_Pow223, sortPriority=-500}
   ))
  n19 = ScalarParameter() {defaultValue=0.7, group=Wetlands_MossyForest, parameterName=Moss_Noise1_Pow223_Add, sortPriority=-500}
  n20 = Multiply(A=LinearInterpolate(A=@n15.RGB, B=Multiply(A=@n17.RGB, B=@n15.RGB), Alpha=@n16), B=Saturate(None=Add(A=@n18, B=@n19)))
  LinearInterpolate(
   A=LinearInterpolate(A=@n16, B=@n20, Alpha=ScalarParameter() {defaultValue=1, group=Wetlands_MossyForest, parameterName=Moss_Noise1_DEBUG01, sortPriority=-500})
   B=@n15.RGB
   Alpha=ScalarParameter() {group=Wetlands_MossyForest, parameterName=Moss_Noise1_DEBUG033, sortPriority=-500}
  )
  LinearInterpolate(B=@n15.A, Alpha=ScalarParameter() {defaultValue=1, group=Displacement, parameterName=Moss_DisplacementAmt, sortPriority=0}) {constA=0.5}
  MaterialFunctionCall(InputMaterial=@n1, LayerMaterial=@n2, LayerWeight=@n3, UseContrast=@n4, Contrast=@n5) {materialFunction=&MF_HeightBlendLandscapeLayers_OpacityMask}
  n21 = SetMaterialAttributes!(MaterialAttributes, BaseColor, Specular, Roughness, Normal, OpacityMask, Displacement) {attributeSetTypes [
    #1
    #2
    #3
    #4
    #5
    #6
   ]}
  n22 = TextureSampleParameter2D(UVs=@n14.Result) {
   group=MossyForest
   parameterName=Hermes_Moss_Spec
   samplerSource=SSM_Wrap_WorldGroupSettings
   samplerType=SAMPLERTYPE_Masks
   texture=/Game/Environments/GlobalTextures/T_Hermes_MossyGround_01_S
  }
  n23 = TextureSampleParameter2D(UVs=@n14.Result) {
   group=MossyForest
   parameterName=Hermes_Moss_Normal
   samplerSource=SSM_Wrap_WorldGroupSettings
   samplerType=SAMPLERTYPE_Normal
   texture=/Game/Environments/GlobalTextures/T_Hermes_MossyGround_01_N
  }
  n24 = SetMaterialAttributes(
   BaseColor=@n20
   Specular=@n22.R
   Roughness=@n22.R
   Normal=MaterialFunctionCall!(
    Normals=@n23.RGB
    TransformToTangentSpace=NamedRerouteUsage()
    Normal Strength=ScalarParameter() {defaultValue=1, group=Wetlands_MossyForest, parameterName=Moss_NormalStrength, sortPriority=-500}
   ) {materialFunction=/Game/Athena/Apollo/Environments/Landscape/Materials/Biome/CommonMaterials/MF_HQ_WorldAlignedNormal}.Result
   OpacityMask=@n15.A
   Displacement=@n15.A
  ) {attributeSetTypes [
    #1
    #2
    #3
    #4
    #5
    #6
   ]}
  n25 = NamedRerouteUsage()
  MaterialFunctionCall(InputMaterial=@n21, LayerMaterial=@n24, LayerWeight=@n25) {materialFunction=&MF_HeightBlendLandscapeLayers_OpacityMask}
  n26 = FunctionInput(Preview=StaticBool()) {
   bUsePreviewValueAsDefault=true
   description="Default = False"
   inputName=UseTangentNormals
   inputType=FunctionInput_StaticBool
   sortPriority=15
  }
  NamedRerouteDeclaration(None=@n26) {name=UseTangentNormalsBool, nodeColor=<rgba 0.5 0 1 1>}
  NamedRerouteDeclaration(None=FunctionInput(Preview=Constant()) {bUsePreviewValueAsDefault=true, inputName=VariationMasks Mobile, inputType=FunctionInput_Vector4, sortPriority=40}) {name=VariationMasks Mobile, nodeColor=<rgba 1 0.8498 0.0704 1>}
  NamedRerouteDeclaration(None=@n24) {name=LayerMaterial, nodeColor=<rgba 0 1 0 1>}
  NamedRerouteDeclaration(None=FunctionInput(Preview=Constant()) {bUsePreviewValueAsDefault=true, inputName=LayerWeight, inputType=FunctionInput_Scalar, sortPriority=15}) {name=LayerWeight, nodeColor=<rgba 0.5 0 1 1>}
  NamedRerouteDeclaration(None=@n6.LayerMask) {name=LayerMask, nodeColor=<rgba 0.7344 0 1 1>}
  n27 = FunctionInput(Preview=Constant()) {bUsePreviewValueAsDefault=true, inputName=TerrainMasks_QuadA, inputType=FunctionInput_Vector4, sortPriority=20}
  NamedRerouteDeclaration(None=ComponentMask(None=NamedRerouteDeclaration(None=@n27) {name=TerrainMasks_QuadA, nodeColor=<rgba 0.5 0 1 1>}) {b=true}) {name=BM_MossyForest, nodeColor=<rgba 0.5 0 1 1>}
  n28 = SetMaterialAttributes!(MaterialAttributes)
  NamedRerouteDeclaration(None=FunctionInput(Preview=@n28) {bUsePreviewValueAsDefault=true, inputName=Material, inputType=FunctionInput_MaterialAttributes}) {name=InputMaterial, nodeColor=<rgba 0 1 0 1>}
  NamedRerouteDeclaration(None=FunctionInput(Preview=Constant()) {bUsePreviewValueAsDefault=true, inputName=TerrainMasks_QuadC, inputType=FunctionInput_Vector4, sortPriority=30}) {name=TerrainMasks_QuadC, nodeColor=<rgba 0.5 0 1 1>}
  NamedRerouteDeclaration(None=@n8.LayerMask) {name=MobileLayerMask, nodeColor=<rgba 0 0.9062 1 1>}
  NamedRerouteDeclaration(None=FunctionInput(Preview=Constant()) {bUsePreviewValueAsDefault=true, inputName=TerrainMasks_QuadD, inputType=FunctionInput_Vector4, sortPriority=35}) {name=TerrainMasks_QuadD, nodeColor=<rgba 0.5 0 1 1>}
  NamedRerouteDeclaration(None=FunctionInput(Preview=Constant()) {bUsePreviewValueAsDefault=true, inputName=TerrainMasks_QuadB, inputType=FunctionInput_Vector4, sortPriority=25}) {name=TerrainMasks_QuadB, nodeColor=<rgba 0.5 0 1 1>}
  NamedRerouteDeclaration(None=FunctionInput(Preview=@n28) {bUsePreviewValueAsDefault=true, inputName=MobileMaterial, inputType=FunctionInput_MaterialAttributes, sortPriority=10}) {name=MobileInputMaterial, nodeColor=<rgba 0 0.5 1 1>}
  NamedRerouteDeclaration(None=MaterialFunctionCall!(In=NamedRerouteUsage()) {materialFunction=/Game/Hermes/Terrain/Landscape/Materials/Layers/QualityLevel/MF_Hermes_Moss_NS}.Result) {name=MobileLayerMaterial, nodeColor=<rgba 0 0.5 1 1>}
  NamedRerouteDeclaration(None=StaticBool() {value=true}) {name=UseContrast, nodeColor=<rgba 0.5 0 1 1>}
  NamedRerouteDeclaration(None=ScalarParameter() {defaultValue=0.5, group=Brimstone01, parameterName=Brimstone01LayerBlendContrast, sortPriority=500}) {name=LayerBlendContrast, nodeColor=<rgba 0.5 0 1 1>}
  SetMaterialAttributes()
  StaticSwitch(
   True=BlendMaterialAttributes(A=@n21, B=@n24, Alpha=@n25) {desc="Vertex blend type altered!", vertexAttributeBlendType=UseA}
   False=StaticSwitch(
    True=@n21
    False=@n24
    Value=StaticBoolParameter() {defaultValue=true, desc="False=Use_MossyForest", group=LayerUsage, parameterName=Use_Wetlands, sortPriority=-500}
   )
   Value=StaticBoolParameter() {desc="False=Use individual layers", group=LayerUsage, parameterName=Blend_Wetlands_to_MossyForest, sortPriority=-500}
  )
 ]}
